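# Animation Blueprint for the female medium UltraRare Commando 01 body cosmetic, adding secondary motion to gear bones.
# On initialization it caches the owner's skeletal mesh and copies that pose.
# Anim Dynamics then simulates the chest tube and both shoulder plates in sequence, with plate alpha scaled by 2D pawn speed.
# Two grenade dynamics nodes with pelvis sphere limits exist but are not wired to the output pose.
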AnimBlueprint /BRCosmetics/Characters/Player/Female/Medium/Bodies/F_MED_UltraRareCommando_01/Meshes/F_MED_Commando_UltraRare_01_Body_Skeleton_AnimBP
target_skeleton=&/F_MED_Commando_UltraRare_01_Body_Skeleton
graphs [
 AnimGraph {nodes [
   n1 = Anim Dynamics - Bone: dyn_grenade_a(Alpha=1, ExternalForce="0,0,0") {
    angularSpringConstant=1000
    bAngularSpring=true
    bGravityOverrideInSimSpace=true
    numSolverIterationsPostUpdate=4
    numSolverIterationsPreUpdate=16
    alphaBoolBlend {blendOption=Linear}
    boundBone {boneName=dyn_grenade_a}
    physicsBodyDefinitions [{
      sphereCollisionRadius=0
      boundBone {boneName=dyn_grenade_a}
      boxExtents=<10 10 20>
      constraintSetup {
       angularTargetAxis=AxisY
       twistAxis=AxisZ
       angularLimitsMax=<20 0 0>
       angularLimitsMin=<0 -20 0>
       angularTarget=<0 0 0>
       linearAxesMax=<0 0 0.5>
       linearAxesMin=<0 0 -0.5>
      }
      localJointOffset=<0 0 -2>
     }]
    sphericalLimits [{limitRadius=16, drivingBone {boneName=pelvis}, sphereLocalOffset=<-1 1 0>}]
   }
   Anim Dynamics - Bone: dyn_grenade_b(ComponentPose=@n1, Alpha=1, ExternalForce="0,0,0") {
    angularSpringConstant=5000
    bGravityOverrideInSimSpace=true
    gravityScale=0.2
    numSolverIterationsPostUpdate=4
    numSolverIterationsPreUpdate=16
    alphaBoolBlend {blendOption=Linear}
    boundBone {boneName=dyn_grenade_b}
    physicsBodyDefinitions [{
      sphereCollisionRadius=0
      boundBone {boneName=dyn_grenade_b}
      boxExtents=<10 10 20>
      constraintSetup {
       angularTargetAxis=AxisY
       twistAxis=AxisZ
       angularLimitsMax=<20 0 0>
       angularLimitsMin=<0 -20 0>
       angularTarget=<0 0 0>
       linearAxesMax=<0 0 0.5>
       linearAxesMin=<0 0 -0.5>
      }
      localJointOffset=<0 0 -2>
     }]
    sphericalLimits [{limitRadius=16, drivingBone {boneName=pelvis}, sphereLocalOffset=<-1 1 0>}]
   }
   n2 = Get MeshToCopy() -> (MeshToCopy:Skeletal Mesh Component Object Reference)
   n3 = Anim Dynamics - Bone: dyn_chest_tube(ComponentPose=Local To Component(LocalPose=Copy Pose From Mesh(SourceMeshComponent=@n2)), Alpha=1, ExternalForce="0,0,0") {
    bGravityOverrideInSimSpace=true
    gravityScale=0.01
    numSolverIterationsPostUpdate=4
    numSolverIterationsPreUpdate=16
    alphaBoolBlend {blendOption=Linear}
    boundBone {boneName=dyn_chest_tube}
    physicsBodyDefinitions [{
      sphereCollisionRadius=0
      boundBone {boneName=dyn_chest_tube}
      boxExtents=<20 2.5 5.5>
      constraintSetup {
       angularTargetAxis=AxisZ
       twistAxis=AxisZ
       angularLimitsMax=<25 45 0>
       angularLimitsMin=<0 -45 0>
       angularTarget=<0 0 0>
       linearAxesMax=<1.5 1.5 1.5>
       linearAxesMin=<-1.5 -1.5 -1.5>
      }
      localJointOffset=<0 0 -0.4>
     }]
   }
   n4 = float * float(A=MapRangeClamped(Value=Get PawnSpeed2D(), InRangeA=0, InRangeB=100, OutRangeA=.2), B=1)
   n5 = Anim Dynamics - Bone: dyn_shoulder_plate_r(ComponentPose=@n3, Alpha=@n4, ExternalForce="0,0,0") {
    bGravityOverrideInSimSpace=true
    gravityScale=0.2
    alphaBoolBlend {blendOption=Linear}
    boundBone {boneName=dyn_shoulder_plate_r}
    physicsBodyDefinitions [{
      sphereCollisionRadius=0
      boundBone {boneName=dyn_shoulder_plate_r}
      boxExtents=<20 20 20>
      constraintSetup {angularLimitsMax=<5 5 11.7924>, angularLimitsMin=<-5 -5 -11.7925>, angularTarget=<0 0 0>, linearAxesMax=<2 1 1>, linearAxesMin=<0 -1 -1>}
      localJointOffset=<-3 0 0>
     }]
   }
   n6 = Anim Dynamics - Bone: dyn_shoulder_plate_l(ComponentPose=@n5, Alpha=@n4, ExternalForce="0,0,0") {
    bGravityOverrideInSimSpace=true
    gravityScale=0.2
    alphaBoolBlend {blendOption=Linear}
    boundBone {boneName=dyn_shoulder_plate_l}
    physicsBodyDefinitions [{
      sphereCollisionRadius=0
      boundBone {boneName=dyn_shoulder_plate_l}
      boxExtents=<20 20 20>
      constraintSetup {angularLimitsMax=<5 5 11.7924>, angularLimitsMin=<-5 -5 -11.7925>, angularTarget=<0 0 0>, linearAxesMax=<0 1 1>, linearAxesMin=<-2 -1 -1>}
      localJointOffset=<3 0 0>
     }]
   }
   Output Pose(Result=Component To Local(ComponentPose=@n6)) {layerGroup=None}
  ]}
 EventGraph {nodes [
   Event BlueprintInitializeAnimation()
   > Set MeshToCopy(MeshToCopy:Skeletal Mesh Component Object Reference=GetOwnerSkeletalMeshComponent()) -> (Output_Get:Skeletal Mesh Component Object Reference)
   Event BlueprintUpdateAnimation()
  ]}
]
variables [MeshToCopy {type=SkeletalMeshComponent}]
